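# Animation Blueprint for the OutlineLove glider, animating its closed, deploy and gliding states during skydiving and in cosmetic preview.
# Separate gameplay and front-end locomotion state machines are selected by IsCosmeticPreview, each using curve-driven Body_Offset scaling.
# Gameplay adds directional lean and aim-driven turn layers from SkydivingState.
# Both paths then receive additive jitter, wing sway, head darts and caws, while the event graph interpolates Jitter from forward acceleration.
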AnimBlueprint /BRCosmetics/Animation/Game/MainPlayer/Skydive/ParaGlide/OutlineLove/OutlineLove_AnimBP
target_skeleton=/BRCosmetics/Gadgets/Assets/VinderTech_GliderChute/Glider_OutlineLove/Meshes/OutlineLove_Glider_Skeleton
graphs [
 AnimGraph {nodes [
   Output Pose(Result=Use cached pose 'Additive Animations'()) {layerGroup=None}
   AnimGraphNode_StateMachine_1 = Locomotion() {is=title AnimGraphNode_StateMachine}
   n1 = "Transform (Modify) Bone - Bone: Body_Offset"!(
    ComponentPose=Local To Component(LocalPose=@AnimGraphNode_StateMachine_1)
    AlphaCurveName:Name=Scale
    Translation="0,0,0"
    Rotation="0,0,0"
    Scale="0.1,0.1,0.1"
   ) -> (Pose) {alphaInputType=Curve, scaleMode=BMM_Replace, boneToModify {boneName=Body_Offset}, scale=<0.1 0.1 0.1>}
   n2 = Component To Local(ComponentPose=@n1)
   AnimGraphNode_StateMachine_2 = Lean Transitions() {is=title AnimGraphNode_StateMachine}
   n3 = Break Fort Anim Input Skydiving!(FortAnimInput_Skydiving:"Fort Anim Input Skydiving Structure (by ref)"=Get SkydivingState() -> (SkydivingState:Fort Anim Input Skydiving Structure)) -> (
    ParaGilderRootOffset:Vector=~
    ParaGliderRootRotationOverride:Rotator=~
    bUseParaGlideRootModifier:bool=~
    bIsSkydivingFromLaunchPad:bool=~
    bIsSkydivingFromBus:bool=~
    bIsInVortex:bool=~
    bIsDiving:bool=~
    bIsParachuteOpen:bool=~
    bIsSkydiving:bool=~
    bIsParachuteLeaning:bool=~
    bIsSkydiveLeaning:bool=~
    bIsLeaning:bool=~
    bIsSkydiveDiveMode:bool=~
    bPlaySkydiveDrift:bool=~
    bSkydiveDriftDelayActive:bool=~
    bSkydiveDriftAnimAllowed:bool=~
    bIsGliderRight:bool=~
    bIsGliderCenter:bool=~
    bIsGliderLeft:bool=~
    bIsGliderForward:bool=~
    bIsGliderBack:bool=~
    bHasDeployedAndIsStandingUpright:bool=~
    LocalAccelForward:float=~
    LocalAccelRight:float=~
    LocalVelocityRight:float=~
    SkydiveAimPitch:float=~
    SkydiveAimYaw:float=~
    DeployChuteAnimRate:float=~
    SkydiveDriftAnimRate:float=~
    SkydiveDriftAnimRateCurrent:float=~
    SkydiveFidgetAnimRate:float=~
    SkydiveFidgetAnimRateCurrent:float=~
    SkydiveAdditiveAlpha:float=~
    SkydiveDriftDelay:float=~
    SkydiveDriftAnim:int
    SkydiveDriftAnimMax:int
    LaunchpadAnim:int
    LocalAccelDir:ESkydivingDirection Enum
    DirectionLast:ESkydivingDirection Enum
    PoseAdditive:Anim Sequence Object Reference
   ) {is=title K2Node_BreakStruct}
   Save cached pose 'GamePlay Locomotion'(Pose=Apply Additive!(
     Base=Apply Additive(Base=@n2, Additive=@AnimGraphNode_StateMachine_2, AlphaCurveName=paraglide_lean) {alphaInputType=Curve, alphaBoolBlend {blendOption=Linear}, alphaScaleBiasClamp {bias=1, scale=-1}}
     Additive=Blendspace Player 'BS_OutlineLove_Turn_GLIDER'(X=@n3.SkydiveAimYaw, Y=@n3.SkydiveAimPitch) {blendSpace=&/BS_OutlineLove_Turn_GLIDER}
     AlphaCurveName:Name=paraglide_lean
    ) -> (Pose) {alphaInputType=Curve, alphaBoolBlend {blendOption=Linear}, alphaScaleBiasClamp {bias=1, scale=-1}}) {cacheName=GamePlay Locomotion}
   AnimGraphNode_StateMachine_3 = FE Locomotion() {is=title AnimGraphNode_StateMachine}
   n4 = "Transform (Modify) Bone - Bone: Body_Offset"!(
    ComponentPose=Local To Component(LocalPose=@AnimGraphNode_StateMachine_3)
    bAlphaBoolEnabled:bool=True
    Translation="-50,100,0"
    Rotation="0,0,0"
    Scale="1,1,1"
   ) -> (Pose) {alphaInputType=Bool, translationMode=BMM_Additive, boneToModify {boneName=Body_Offset}, translation=<-50 100>}
   n5 = "Transform (Modify) Bone - Bone: Body_Offset"!(ComponentPose=@n4, AlphaCurveName:Name=scale, Translation="0,0,0", Rotation="0,0,0", Scale="0.1,0.1,0.1") -> (Pose) {alphaInputType=Curve, scaleMode=BMM_Replace, boneToModify {boneName=Body_Offset}, scale=<0.1 0.1 0.1>}
   Save cached pose 'FrontEnd Locomotion'(Pose=Component To Local(ComponentPose=@n5)) {cacheName=FrontEnd Locomotion}
   n6 = Use cached pose 'FrontEnd Locomotion'()
   n7 = Use cached pose 'GamePlay Locomotion'()
   n8 = Get IsCosmeticPreview!(self:Fort Player Parachute Object Reference=Get OwnerGlider() -> (OwnerGlider:Fort Player Parachute Object Reference)) -> (IsCosmeticPreview:bool)
   n9 = Apply Additive(
    Base=Apply Additive!(
     Base=Blend Poses by bool(BlendPose_0=@n6, BlendPose_1=@n7, bActiveValue=@n8) {blendType=Linear}
     Additive="Sequence Player (additive) 'OutlineLove_Jitter_Anim_GLIDER'"!(PlayRate:float=1) -> (Pose) {sequence=&/OutlineLove_Jitter_Anim_GLIDER}
     bAlphaBoolEnabled:bool=True
    ) -> (Pose) {alphaInputType=Bool, alphaBoolBlend {blendOption=Linear}, alphaScaleBiasClamp {bias=1, scale=-1}}
    Additive="Sequence Player (additive) 'OutlineLove_Wing_Sway_Anim_GLIDER'"() {sequence=&/OutlineLove_Wing_Sway_Anim_GLIDER}
    AlphaCurveName=WingSway
   ) {alphaInputType=Curve, alphaBoolBlend {blendOption=Linear}, alphaScaleBiasClamp {bias=1, scale=-1}}
   AnimGraphNode_StateMachine_5 = Head Darts() {is=title AnimGraphNode_StateMachine}
   AnimGraphNode_StateMachine_4 = Roar() {is=title AnimGraphNode_StateMachine}
   Save cached pose 'Additive Animations'(Pose=Apply Additive!(
     Base=Apply Additive(Base=@n9, Additive=@AnimGraphNode_StateMachine_5, AlphaCurveName=SecondaryActions) {alphaInputType=Curve, alphaBoolBlend {blendOption=Linear}, alphaScaleBiasClamp {bias=1, scale=-1}}
     Additive=@AnimGraphNode_StateMachine_4
     AlphaCurveName:Name=SecondaryActions
    ) -> (Pose) {alphaInputType=Curve, alphaBoolBlend {blendOption=Linear}, alphaScaleBiasClamp {bias=1, scale=-1}}) {cacheName=Additive Animations}
  ]}
 Locomotion {
  entry=Closed
  path=AnimGraph.AnimGraphNode_StateMachine_1.Locomotion
  ids=id
  nodes [
   AnimStateNode_0 = Idle_Open() {is=title AnimStateNode}
   AnimStateNode_2 = Closed() {is=title AnimStateNode, stateEntered {guid=#1, notifyName=ClosedEntered}}
   AnimStateNode_3 = Open_Into() {is=title AnimStateNode, stateEntered {guid=#2, notifyName=DeployEntered}}
   AnimStateTransitionNode_0 = AnimStateTransitionNode() {bAutomaticRuleBasedOnSequencePlayerInState=true, blendMode=Linear, crossfadeDuration=0}
   AnimStateTransitionNode_1 = AnimStateTransitionNode() {crossfadeDuration=0}
   AnimStateTransitionNode_2 = AnimStateTransitionNode() {crossfadeDuration=0}
   AnimStateTransitionNode_3 = AnimStateTransitionNode() {crossfadeDuration=0, priorityOrder=2}
  ]
  links [
   {
    dst=Open_Into
    rule=AnimGraph.AnimGraphNode_StateMachine_1.Locomotion.AnimStateTransitionNode_1.Transition
    src=Closed
    via=AnimStateTransitionNode_1
   }
   {
    dst=Closed
    rule=AnimGraph.AnimGraphNode_StateMachine_1.Locomotion.AnimStateTransitionNode_2.Transition
    src=Idle_Open
    via=AnimStateTransitionNode_2
   }
   {
    dst=Closed
    rule=AnimGraph.AnimGraphNode_StateMachine_1.Locomotion.AnimStateTransitionNode_3.Transition
    src=Open_Into
    via=AnimStateTransitionNode_3
   }
   {
    dst=Idle_Open
    rule=AnimGraph.AnimGraphNode_StateMachine_1.Locomotion.AnimStateTransitionNode_0.Transition
    src=Open_Into
    via=AnimStateTransitionNode_0
   }
  ]
 }
 Idle_Open {
  path=AnimGraph.AnimGraphNode_StateMachine_1.Locomotion.AnimStateNode_0.Idle_Open
  nodes [
   n1 = Break Fort Anim Input Skydiving(FortAnimInput_Skydiving=Get SkydivingState()) -> (
    ParaGilderRootOffset=~
    ParaGliderRootRotationOverride=~
    bUseParaGlideRootModifier=~
    bIsSkydivingFromLaunchPad=~
    bIsSkydivingFromBus=~
    bIsInVortex=~
    bIsDiving=~
    bIsParachuteOpen=~
    bIsSkydiving=~
    bIsParachuteLeaning=~
    bIsSkydiveLeaning=~
    bIsLeaning=~
    bIsSkydiveDiveMode=~
    bPlaySkydiveDrift=~
    bSkydiveDriftDelayActive=~
    bSkydiveDriftAnimAllowed=~
    bIsGliderRight=~
    bIsGliderCenter=~
    bIsGliderLeft=~
    bIsGliderForward=~
    bIsGliderBack=~
    bHasDeployedAndIsStandingUpright=~
    LocalAccelForward=~
    LocalAccelRight=~
    LocalVelocityRight=~
    SkydiveAimPitch=~
    SkydiveAimYaw=~
    DeployChuteAnimRate=~
    SkydiveDriftAnimRate=~
    SkydiveDriftAnimRateCurrent=~
    SkydiveFidgetAnimRate=~
    SkydiveFidgetAnimRateCurrent=~
    SkydiveAdditiveAlpha=~
    SkydiveDriftDelay=~
   )
   Output Animation Pose(Result=Blendspace Player 'BS_OutlineLove_GLIDER'(X=@n1.LocalAccelRight, Y=@n1.LocalAccelForward) {blendSpace=&/BS_OutlineLove_GLIDER})
  ]
 }
 Closed {
  path=AnimGraph.AnimGraphNode_StateMachine_1.Locomotion.AnimStateNode_2.Closed
  nodes [Output Animation Pose(Result=Sequence Player 'OutlineLove_Closed_Pose_GLIDER'() {bLoopAnimation=false, sequence=&/OutlineLove_Closed_Pose_GLIDER})]
 }
 Open_Into {
  path=AnimGraph.AnimGraphNode_StateMachine_1.Locomotion.AnimStateNode_3.Open_Into
  nodes [
   n1 = Break Fort Anim Input Skydiving(FortAnimInput_Skydiving=Get SkydivingState()) -> (
    ParaGilderRootOffset=~
    ParaGliderRootRotationOverride=~
    bUseParaGlideRootModifier=~
    bIsSkydivingFromLaunchPad=~
    bIsSkydivingFromBus=~
    bIsInVortex=~
    bIsDiving=~
    bIsParachuteOpen=~
    bIsSkydiving=~
    bIsParachuteLeaning=~
    bIsSkydiveLeaning=~
    bIsLeaning=~
    bIsSkydiveDiveMode=~
    bPlaySkydiveDrift=~
    bSkydiveDriftDelayActive=~
    bSkydiveDriftAnimAllowed=~
    bIsGliderRight=~
    bIsGliderCenter=~
    bIsGliderLeft=~
    bIsGliderForward=~
    bIsGliderBack=~
    bHasDeployedAndIsStandingUpright=~
    LocalAccelForward=~
    LocalAccelRight=~
    LocalVelocityRight=~
    SkydiveAimPitch=~
    SkydiveAimYaw=~
    DeployChuteAnimRate=~
    SkydiveDriftAnimRate=~
    SkydiveDriftAnimRateCurrent=~
    SkydiveFidgetAnimRate=~
    SkydiveFidgetAnimRateCurrent=~
    SkydiveAdditiveAlpha=~
    SkydiveDriftDelay=~
   )
   Output Animation Pose(Result=Blendspace Player 'BS_OutlineLove_Into_GLIDER'(X=@n1.SkydiveAimYaw, Y=@n1.SkydiveAimPitch) {bLoop=false, blendSpace=&/BS_OutlineLove_Into_GLIDER})
  ]
 }
 Transition {
  path=AnimGraph.AnimGraphNode_StateMachine_1.Locomotion.AnimStateTransitionNode_0.Transition
  nodes [
   Result()
   "Equal (Float)"!(A:double="Get Relevant Anim Time Remaining (Open_Into)"!() -> (ReturnValue:float) {is=title K2Node_AnimGetter}, B:double=0) -> (ReturnValue:bool) {is=title K2Node_CallFunction}
  ]
 }
 Transition {
  path=AnimGraph.AnimGraphNode_StateMachine_1.Locomotion.AnimStateTransitionNode_1.Transition
  nodes [
   n1 = Get SkydivingState()
   Result(bCanEnterTransition=OR Boolean(
     A=Break Fort Anim Input Skydiving(FortAnimInput_Skydiving=@n1) -> (
      ParaGilderRootOffset=~
      ParaGliderRootRotationOverride=~
      bUseParaGlideRootModifier=~
      bIsSkydivingFromLaunchPad=~
      bIsSkydivingFromBus=~
      bIsInVortex=~
      bIsDiving=~
      bIsParachuteOpen=~
      bIsSkydiving=~
      bIsParachuteLeaning=~
      bIsSkydiveLeaning=~
      bIsLeaning=~
      bIsSkydiveDiveMode=~
      bPlaySkydiveDrift=~
      bSkydiveDriftDelayActive=~
      bSkydiveDriftAnimAllowed=~
      bIsGliderRight=~
      bIsGliderCenter=~
      bIsGliderLeft=~
      bIsGliderForward=~
      bIsGliderBack=~
      bHasDeployedAndIsStandingUpright=~
      LocalAccelForward=~
      LocalAccelRight=~
      LocalVelocityRight=~
      SkydiveAimPitch=~
      SkydiveAimYaw=~
      DeployChuteAnimRate=~
      SkydiveDriftAnimRate=~
      SkydiveDriftAnimRateCurrent=~
      SkydiveFidgetAnimRate=~
      SkydiveFidgetAnimRateCurrent=~
      SkydiveAdditiveAlpha=~
      SkydiveDriftDelay=~
     ).bIsParachuteOpen
     B=Get bForceOpen() -> (bForceOpen:bool)
    ))
  ]
 }
 Transition {
  path=AnimGraph.AnimGraphNode_StateMachine_1.Locomotion.AnimStateTransitionNode_2.Transition
  nodes [
   n1 = Get bForceOpen()
   n2 = Break Fort Anim Input Skydiving(FortAnimInput_Skydiving=Get SkydivingState()) -> (
    ParaGilderRootOffset=~
    ParaGliderRootRotationOverride=~
    bUseParaGlideRootModifier=~
    bIsSkydivingFromLaunchPad=~
    bIsSkydivingFromBus=~
    bIsInVortex=~
    bIsDiving=~
    bIsParachuteOpen=~
    bIsSkydiving=~
    bIsParachuteLeaning=~
    bIsSkydiveLeaning=~
    bIsLeaning=~
    bIsSkydiveDiveMode=~
    bPlaySkydiveDrift=~
    bSkydiveDriftDelayActive=~
    bSkydiveDriftAnimAllowed=~
    bIsGliderRight=~
    bIsGliderCenter=~
    bIsGliderLeft=~
    bIsGliderForward=~
    bIsGliderBack=~
    bHasDeployedAndIsStandingUpright=~
    LocalAccelForward=~
    LocalAccelRight=~
    LocalVelocityRight=~
    SkydiveAimPitch=~
    SkydiveAimYaw=~
    DeployChuteAnimRate=~
    SkydiveDriftAnimRate=~
    SkydiveDriftAnimRateCurrent=~
    SkydiveFidgetAnimRate=~
    SkydiveFidgetAnimRateCurrent=~
    SkydiveAdditiveAlpha=~
    SkydiveDriftDelay=~
   )
   Result(bCanEnterTransition=AND Boolean(A=NOT Boolean(A=@n2.bIsParachuteOpen), B=NOT Boolean(A=@n1)))
  ]
 }
 Transition {
  path=AnimGraph.AnimGraphNode_StateMachine_1.Locomotion.AnimStateTransitionNode_3.Transition
  nodes [
   n1 = Get bForceOpen()
   n2 = Break Fort Anim Input Skydiving(FortAnimInput_Skydiving=Get SkydivingState()) -> (
    ParaGilderRootOffset=~
    ParaGliderRootRotationOverride=~
    bUseParaGlideRootModifier=~
    bIsSkydivingFromLaunchPad=~
    bIsSkydivingFromBus=~
    bIsInVortex=~
    bIsDiving=~
    bIsParachuteOpen=~
    bIsSkydiving=~
    bIsParachuteLeaning=~
    bIsSkydiveLeaning=~
    bIsLeaning=~
    bIsSkydiveDiveMode=~
    bPlaySkydiveDrift=~
    bSkydiveDriftDelayActive=~
    bSkydiveDriftAnimAllowed=~
    bIsGliderRight=~
    bIsGliderCenter=~
    bIsGliderLeft=~
    bIsGliderForward=~
    bIsGliderBack=~
    bHasDeployedAndIsStandingUpright=~
    LocalAccelForward=~
    LocalAccelRight=~
    LocalVelocityRight=~
    SkydiveAimPitch=~
    SkydiveAimYaw=~
    DeployChuteAnimRate=~
    SkydiveDriftAnimRate=~
    SkydiveDriftAnimRateCurrent=~
    SkydiveFidgetAnimRate=~
    SkydiveFidgetAnimRateCurrent=~
    SkydiveAdditiveAlpha=~
    SkydiveDriftDelay=~
   )
   Result(bCanEnterTransition=AND Boolean(A=NOT Boolean(A=@n2.bIsParachuteOpen), B=NOT Boolean(A=@n1)))
  ]
 }
 Lean Transitions {
  entry=Lean_Zero
  path=AnimGraph.AnimGraphNode_StateMachine_2.Lean Transitions
  ids=id
  nodes [
   AnimStateNode_0 = Lean_Zero() {
    is=title AnimStateNode
    stateEntered {guid=#3}
    stateFullyBlended {guid=#4}
    stateLeft {guid=#5}
   }
   AnimStateNode_2 = Lean_Out() {is=title AnimStateNode}
   AnimStateNode_5 = Lean_Into() {
    is=title AnimStateNode
    stateEntered {guid=#6}
    stateFullyBlended {guid=#7}
    stateLeft {guid=#8}
   }
   AnimStateTransitionNode_0 = AnimStateTransitionNode()
   AnimStateTransitionNode_1 = AnimStateTransitionNode() {bAllowInertializationForSelfTransitions=true}
   AnimStateTransitionNode_11 = AnimStateTransitionNode()
   AnimStateTransitionNode_18 = AnimStateTransitionNode() {bAutomaticRuleBasedOnSequencePlayerInState=true, blendMode=Linear, priorityOrder=2}
  ]
  links [
   {
    dst=Lean_Out
    rule=AnimGraph.AnimGraphNode_StateMachine_2.Lean Transitions.AnimStateTransitionNode_11.Transition
    src=Lean_Into
    via=AnimStateTransitionNode_11
   }
   {
    dst=Lean_Into
    rule=AnimGraph.AnimGraphNode_StateMachine_2.Lean Transitions.AnimStateTransitionNode_1.Transition
    src=Lean_Out
    via=AnimStateTransitionNode_1
   }
   {
    dst=Lean_Zero
    rule=AnimGraph.AnimGraphNode_StateMachine_2.Lean Transitions.AnimStateTransitionNode_18.Transition
    src=Lean_Out
    via=AnimStateTransitionNode_18
   }
   {
    dst=Lean_Into
    rule=AnimGraph.AnimGraphNode_StateMachine_2.Lean Transitions.AnimStateTransitionNode_0.Transition
    src=Lean_Zero
    via=AnimStateTransitionNode_0
   }
  ]
 }
 Lean_Zero {
  path=AnimGraph.AnimGraphNode_StateMachine_2.Lean Transitions.AnimStateNode_0.Lean_Zero
  nodes [Output Animation Pose(Result="Sequence Player (additive) 'OutlineLove_Lean_Center_Pose_GLIDER'"() {sequence=&/OutlineLove_Lean_Center_Pose_GLIDER})]
 }
 Lean_Out {
  path=AnimGraph.AnimGraphNode_StateMachine_2.Lean Transitions.AnimStateNode_2.Lean_Out
  nodes [
   n1 = "Sequence Player (additive) 'OutlineLove_Lean_Back_Out_Anim_GLIDER'"() {bLoopAnimation=false, sequence=&/OutlineLove_Lean_Back_Out_Anim_GLIDER}
   n2 = "Sequence Player (additive) 'OutlineLove_Lean_Right_Out_Anim_GLIDER'"() {bLoopAnimation=false, sequence=&/OutlineLove_Lean_Right_Out_Anim_GLIDER}
   n3 = "Sequence Player (additive) 'OutlineLove_Lean_Center_Pose_GLIDER'"() {bLoopAnimation=false, sequence=&OutlineLove_Lean_Center_Pose_GLIDER}
   n4 = "Sequence Player (additive) 'OutlineLove_Lean_Forward_Out_Anim_GLIDER'"() {bLoopAnimation=false, sequence=&/OutlineLove_Lean_Forward_Out_Anim_GLIDER}
   n5 = "Sequence Player (additive) 'OutlineLove_Lean_Left_Out_Anim_GLIDER'"() {bLoopAnimation=false, sequence=&/OutlineLove_Lean_Left_Out_Anim_GLIDER}
   n6 = "Blend Poses (ESkydivingDirection)"!(
    BlendPose_0:Pose Link Structure=@n3
    BlendPose_1:Pose Link Structure=@n4
    BlendPose_2:Pose Link Structure=@n1
    BlendPose_3:Pose Link Structure=@n5
    BlendPose_4:Pose Link Structure=@n2
    BlendTime_0:float=1
    BlendTime_1:float=1
    BlendTime_2:float=1
    BlendTime_3:float=1
    BlendTime_4:float=1
    ActiveEnumValue:Byte=Break Fort Anim Input Skydiving(FortAnimInput_Skydiving=Get SkydivingState()) -> (
     ParaGilderRootOffset=~
     ParaGliderRootRotationOverride=~
     bUseParaGlideRootModifier=~
     bIsSkydivingFromLaunchPad=~
     bIsSkydivingFromBus=~
     bIsInVortex=~
     bIsDiving=~
     bIsParachuteOpen=~
     bIsSkydiving=~
     bIsParachuteLeaning=~
     bIsSkydiveLeaning=~
     bIsLeaning=~
     bIsSkydiveDiveMode=~
     bPlaySkydiveDrift=~
     bSkydiveDriftDelayActive=~
     bSkydiveDriftAnimAllowed=~
     bIsGliderRight=~
     bIsGliderCenter=~
     bIsGliderLeft=~
     bIsGliderForward=~
     bIsGliderBack=~
     bHasDeployedAndIsStandingUpright=~
     LocalAccelForward=~
     LocalAccelRight=~
     LocalVelocityRight=~
     SkydiveAimPitch=~
     SkydiveAimYaw=~
     DeployChuteAnimRate=~
     SkydiveDriftAnimRate=~
     SkydiveDriftAnimRateCurrent=~
     SkydiveFidgetAnimRate=~
     SkydiveFidgetAnimRateCurrent=~
     SkydiveAdditiveAlpha=~
     SkydiveDriftDelay=~
    ).DirectionLast
   ) -> (Pose:Pose Link Structure) {
    is=title AnimGraphNode_BlendListByEnum
    blendType=Linear
    blendPose [
     {linkId=-1, sourceLinkId=-1}
     {linkId=-1, sourceLinkId=-1}
     {linkId=-1, sourceLinkId=-1}
     {linkId=-1, sourceLinkId=-1}
     {linkId=-1, sourceLinkId=-1}
    ]
    blendTime [0.1,0.1,0.1,0.1,0.1]
   }
   Output Animation Pose(Result=@n6)
  ]
 }
 Lean_Into {
  path=AnimGraph.AnimGraphNode_StateMachine_2.Lean Transitions.AnimStateNode_5.Lean_Into
  nodes [
   AnimGraphNode_StateMachine_0 = Lean_Into() {is=title AnimGraphNode_StateMachine}
   Output Animation Pose(Result=@AnimGraphNode_StateMachine_0)
  ]
 }
 Lean_Into {
  entry=Lean_Zero
  path=AnimGraph.AnimGraphNode_StateMachine_2.Lean Transitions.AnimStateNode_5.Lean_Into.AnimGraphNode_StateMachine_0.Lean_Into
  ids=id
  nodes [
   Left Into() {is=title AnimStateAliasNode, also {aliases [Back_Into,Forward_Into,Lean_Zero,Right_Into]}}
   Forward Into() {is=title AnimStateAliasNode, also {aliases [Back_Into,Lean_Zero,Left_Into,Right_Into]}}
   Right Into() {is=title AnimStateAliasNode, also {aliases [Back_Into,Forward_Into,Lean_Zero,Left_Into]}}
   Back Into() {is=title AnimStateAliasNode, also {aliases [Forward_Into,Lean_Zero,Left_Into,Right_Into]}}
   AnimStateNode_0 = Lean_Zero()
   AnimStateNode_3 = Left_Into() {is=title AnimStateNode}
   AnimStateNode_4 = Right_Into() {is=title AnimStateNode}
   AnimStateNode_5 = Forward_Into() {is=title AnimStateNode}
   AnimStateNode_6 = Back_Into() {is=title AnimStateNode}
   AnimStateTransitionNode_1 = AnimStateTransitionNode()
   AnimStateTransitionNode_2 = AnimStateTransitionNode()
   AnimStateTransitionNode_3 = AnimStateTransitionNode()
   AnimStateTransitionNode_4 = AnimStateTransitionNode()
  ]
  links [
   {
    dst=Back_Into
    rule=AnimGraph.AnimGraphNode_StateMachine_2.Lean Transitions.AnimStateNode_5.Lean_Into.AnimGraphNode_StateMachine_0.Lean_Into.AnimStateTransitionNode_1.AnimationTransitionGraph_0
    src=Back Into
    via=AnimStateTransitionNode_1
   }
   {
    dst=Forward_Into
    rule=AnimGraph.AnimGraphNode_StateMachine_2.Lean Transitions.AnimStateNode_5.Lean_Into.AnimGraphNode_StateMachine_0.Lean_Into.AnimStateTransitionNode_3.AnimationTransitionGraph_0
    src=Forward Into
    via=AnimStateTransitionNode_3
   }
   {
    dst=Left_Into
    rule=AnimGraph.AnimGraphNode_StateMachine_2.Lean Transitions.AnimStateNode_5.Lean_Into.AnimGraphNode_StateMachine_0.Lean_Into.AnimStateTransitionNode_4.AnimationTransitionGraph_0
    src=Left Into
    via=AnimStateTransitionNode_4
   }
   {
    dst=Right_Into
    rule=AnimGraph.AnimGraphNode_StateMachine_2.Lean Transitions.AnimStateNode_5.Lean_Into.AnimGraphNode_StateMachine_0.Lean_Into.AnimStateTransitionNode_2.AnimationTransitionGraph_0
    src=Right Into
    via=AnimStateTransitionNode_2
   }
  ]
 }
 Lean_Zero {
  path=AnimGraph.AnimGraphNode_StateMachine_2.Lean Transitions.AnimStateNode_5.Lean_Into.AnimGraphNode_StateMachine_0.Lean_Into.AnimStateNode_0.Lean_Zero
  nodes [
   n1 = "Sequence Player (additive) 'OutlineLove_Lean_Center_Pose_GLIDER'"() {bLoopAnimation=false, sequence=&OutlineLove_Lean_Center_Pose_GLIDER}
   Output Animation Pose(Result=@n1)
  ]
 }
 Left_Into {
  path=AnimGraph.AnimGraphNode_StateMachine_2.Lean Transitions.AnimStateNode_5.Lean_Into.AnimGraphNode_StateMachine_0.Lean_Into.AnimStateNode_3.Left_Into
  nodes [
   n1 = "Sequence Player (additive) 'OutlineLove_Lean_Left_Into_Anim_GLIDER'"() {bLoopAnimation=false, sequence=&/OutlineLove_Lean_Left_Into_Anim_GLIDER}
   Output Animation Pose(Result=@n1)
  ]
 }
 Right_Into {
  path=AnimGraph.AnimGraphNode_StateMachine_2.Lean Transitions.AnimStateNode_5.Lean_Into.AnimGraphNode_StateMachine_0.Lean_Into.AnimStateNode_4.Right_Into
  nodes [
   n1 = "Sequence Player (additive) 'OutlineLove_Lean_Right_Into_Anim_GLIDER'"() {bLoopAnimation=false, sequence=&/OutlineLove_Lean_Right_Into_Anim_GLIDER}
   Output Animation Pose(Result=@n1)
  ]
 }
 Forward_Into {
  path=AnimGraph.AnimGraphNode_StateMachine_2.Lean Transitions.AnimStateNode_5.Lean_Into.AnimGraphNode_StateMachine_0.Lean_Into.AnimStateNode_5.Forward_Into
  nodes [
   n1 = "Sequence Player (additive) 'OutlineLove_Lean_Forward_Into_Anim_GLIDER'"() {
    bLoopAnimation=false
    sequence=&/OutlineLove_Lean_Forward_Into_Anim_GLIDER
   }
   Output Animation Pose(Result=@n1)
  ]
 }
 Back_Into {
  path=AnimGraph.AnimGraphNode_StateMachine_2.Lean Transitions.AnimStateNode_5.Lean_Into.AnimGraphNode_StateMachine_0.Lean_Into.AnimStateNode_6.Back_Into
  nodes [
   n1 = "Sequence Player (additive) 'OutlineLove_Lean_Back_Into_Anim_GLIDER'"() {bLoopAnimation=false, sequence=&/OutlineLove_Lean_Back_Into_Anim_GLIDER}
   Output Animation Pose(Result=@n1)
  ]
 }
 AnimationTransitionGraph_0 {
  path=AnimGraph.AnimGraphNode_StateMachine_2.Lean Transitions.AnimStateNode_5.Lean_Into.AnimGraphNode_StateMachine_0.Lean_Into.AnimStateTransitionNode_1.AnimationTransitionGraph_0
  nodes [Result(bCanEnterTransition=Break Fort Anim Input Skydiving!(FortAnimInput_Skydiving=Get SkydivingState()) -> (
     bIsGliderRight=~
     bIsGliderCenter=~
     bIsGliderLeft=~
     bIsGliderForward=~
     bIsGliderBack=~
     bHasDeployedAndIsStandingUpright=~
     LocalAccelDir
     PoseAdditive
    ).bIsGliderBack)]
 }
 AnimationTransitionGraph_0 {
  path=AnimGraph.AnimGraphNode_StateMachine_2.Lean Transitions.AnimStateNode_5.Lean_Into.AnimGraphNode_StateMachine_0.Lean_Into.AnimStateTransitionNode_2.AnimationTransitionGraph_0
  nodes [Result(bCanEnterTransition=Break Fort Anim Input Skydiving(FortAnimInput_Skydiving=Get SkydivingState()) -> (bIsGliderRight=~, bIsGliderCenter=~, bIsGliderLeft=~, bIsGliderForward=~, bIsGliderBack=~, bHasDeployedAndIsStandingUpright=~))]
 }
 AnimationTransitionGraph_0 {
  path=AnimGraph.AnimGraphNode_StateMachine_2.Lean Transitions.AnimStateNode_5.Lean_Into.AnimGraphNode_StateMachine_0.Lean_Into.AnimStateTransitionNode_3.AnimationTransitionGraph_0
  nodes [Result(bCanEnterTransition=Break Fort Anim Input Skydiving(FortAnimInput_Skydiving=Get SkydivingState()) -> (bIsGliderRight=~, bIsGliderCenter=~, bIsGliderLeft=~, bIsGliderForward=~, bIsGliderBack=~, bHasDeployedAndIsStandingUpright=~).bIsGliderForward)]
 }
 AnimationTransitionGraph_0 {
  path=AnimGraph.AnimGraphNode_StateMachine_2.Lean Transitions.AnimStateNode_5.Lean_Into.AnimGraphNode_StateMachine_0.Lean_Into.AnimStateTransitionNode_4.AnimationTransitionGraph_0
  nodes [Result(bCanEnterTransition=Break Fort Anim Input Skydiving(FortAnimInput_Skydiving=Get SkydivingState()) -> (bIsGliderRight=~, bIsGliderCenter=~, bIsGliderLeft=~, bIsGliderForward=~, bIsGliderBack=~, bHasDeployedAndIsStandingUpright=~).bIsGliderLeft)]
 }
 Transition {
  path=AnimGraph.AnimGraphNode_StateMachine_2.Lean Transitions.AnimStateTransitionNode_0.Transition
  nodes [Result(bCanEnterTransition=Break Fort Anim Input Skydiving!(FortAnimInput_Skydiving=Get SkydivingState()) -> (
     ParaGilderRootOffset:Vector=~
     ParaGliderRootRotationOverride:Rotator=~
     bUseParaGlideRootModifier:bool=~
     bIsSkydivingFromLaunchPad:bool=~
     bIsSkydivingFromBus:bool=~
     bIsInVortex:bool=~
     bIsDiving:bool=~
     bIsParachuteOpen:bool=~
     bIsSkydiving:bool=~
     bIsParachuteLeaning:bool=~
     bIsSkydiveLeaning:bool=~
     bIsLeaning:bool=~
     bIsSkydiveDiveMode:bool=~
     bPlaySkydiveDrift:bool=~
     bSkydiveDriftDelayActive:bool=~
     bSkydiveDriftAnimAllowed:bool=~
     bIsGliderRight=~
     bIsGliderCenter=~
     bIsGliderLeft=~
     bIsGliderForward=~
     bIsGliderBack=~
     bHasDeployedAndIsStandingUpright=~
     LocalAccelForward:float=~
     LocalAccelRight:float=~
     LocalVelocityRight:float=~
     SkydiveAimPitch:float=~
     SkydiveAimYaw:float=~
     DeployChuteAnimRate:float=~
     SkydiveDriftAnimRate:float=~
     SkydiveDriftAnimRateCurrent:float=~
     SkydiveFidgetAnimRate:float=~
     SkydiveFidgetAnimRateCurrent:float=~
     SkydiveAdditiveAlpha:float=~
     SkydiveDriftDelay:float=~
     SkydiveDriftAnim:int
     SkydiveDriftAnimMax:int
     LaunchpadAnim:int
     LocalAccelDir
     DirectionLast:ESkydivingDirection Enum
     PoseAdditive
    ).bIsLeaning)]
 }
 Transition {
  path=AnimGraph.AnimGraphNode_StateMachine_2.Lean Transitions.AnimStateTransitionNode_1.Transition
  nodes [Result(bCanEnterTransition=Get bIsAccelerating2D())]
 }
 Transition {
  path=AnimGraph.AnimGraphNode_StateMachine_2.Lean Transitions.AnimStateTransitionNode_11.Transition
  nodes [Result(bCanEnterTransition=NOT Boolean(A=Get bIsAccelerating2D()))]
 }
 Transition {path=AnimGraph.AnimGraphNode_StateMachine_2.Lean Transitions.AnimStateTransitionNode_18.Transition, nodes [Result()]}
 FE Locomotion {
  entry=Closed
  path=AnimGraph.AnimGraphNode_StateMachine_3.FE Locomotion
  ids=id
  nodes [
   AnimStateNode_0 = Idle_Open()
   AnimStateNode_2 = Closed() {stateEntered {guid=#1, notifyName=ClosedEntered}}
   AnimStateNode_3 = Open_Into() {stateEntered {guid=#2, notifyName=DeployEntered}}
   AnimStateTransitionNode_0 = AnimStateTransitionNode() {bAutomaticRuleBasedOnSequencePlayerInState=true, blendMode=Linear, crossfadeDuration=0}
   AnimStateTransitionNode_1 = AnimStateTransitionNode() {crossfadeDuration=0}
   AnimStateTransitionNode_2 = AnimStateTransitionNode() {crossfadeDuration=0}
   AnimStateTransitionNode_3 = AnimStateTransitionNode() {crossfadeDuration=0, priorityOrder=2}
  ]
  links [
   {
    dst=Open_Into
    rule=AnimGraph.AnimGraphNode_StateMachine_3.FE Locomotion.AnimStateTransitionNode_1.Transition
    src=Closed
    via=AnimStateTransitionNode_1
   }
   {
    dst=Closed
    rule=AnimGraph.AnimGraphNode_StateMachine_3.FE Locomotion.AnimStateTransitionNode_2.Transition
    src=Idle_Open
    via=AnimStateTransitionNode_2
   }
   {
    dst=Closed
    rule=AnimGraph.AnimGraphNode_StateMachine_3.FE Locomotion.AnimStateTransitionNode_3.Transition
    src=Open_Into
    via=AnimStateTransitionNode_3
   }
   {
    dst=Idle_Open
    rule=AnimGraph.AnimGraphNode_StateMachine_3.FE Locomotion.AnimStateTransitionNode_0.Transition
    src=Open_Into
    via=AnimStateTransitionNode_0
   }
  ]
 }
 Idle_Open {
  path=AnimGraph.AnimGraphNode_StateMachine_3.FE Locomotion.AnimStateNode_0.Idle_Open
  nodes [Output Animation Pose(Result=Sequence Player 'OutlineLove_Glide_FrontEnd_Anim_GLIDER'() {sequence=&/OutlineLove_Glide_FrontEnd_Anim_GLIDER})]
 }
 Closed {
  path=AnimGraph.AnimGraphNode_StateMachine_3.FE Locomotion.AnimStateNode_2.Closed
  nodes [Output Animation Pose(Result=Sequence Player 'OutlineLove_Closed_FrontEnd_Pose_GLIDER'!(bLoopAnimation:bool=True) -> (Pose) {sequence=&/OutlineLove_Closed_FrontEnd_Pose_GLIDER})]
 }
 Open_Into {
  path=AnimGraph.AnimGraphNode_StateMachine_3.FE Locomotion.AnimStateNode_3.Open_Into
  nodes [
   n1 = Sequence Player 'OutlineLove_Into_FrontEnd_Anim_GLIDER'() {bLoopAnimation=false, sequence=&/OutlineLove_Into_FrontEnd_Anim_GLIDER}
   Output Animation Pose(Result=@n1)
  ]
 }
 Transition {
  path=AnimGraph.AnimGraphNode_StateMachine_3.FE Locomotion.AnimStateTransitionNode_0.Transition
  nodes [
   Result()
   "Equal (Float)"(A="Get Relevant Anim Time Remaining (Open_Into)"(), B=0)
  ]
 }
 Transition {
  path=AnimGraph.AnimGraphNode_StateMachine_3.FE Locomotion.AnimStateTransitionNode_1.Transition
  nodes [
   n1 = Get SkydivingState()
   Result(bCanEnterTransition=OR Boolean(
     A=Break Fort Anim Input Skydiving(FortAnimInput_Skydiving=@n1) -> (
      ParaGilderRootOffset=~
      ParaGliderRootRotationOverride=~
      bUseParaGlideRootModifier=~
      bIsSkydivingFromLaunchPad=~
      bIsSkydivingFromBus=~
      bIsInVortex=~
      bIsDiving=~
      bIsParachuteOpen=~
      bIsSkydiving=~
      bIsParachuteLeaning=~
      bIsSkydiveLeaning=~
      bIsLeaning=~
      bIsSkydiveDiveMode=~
      bPlaySkydiveDrift=~
      bSkydiveDriftDelayActive=~
      bSkydiveDriftAnimAllowed=~
      bIsGliderRight=~
      bIsGliderCenter=~
      bIsGliderLeft=~
      bIsGliderForward=~
      bIsGliderBack=~
      bHasDeployedAndIsStandingUpright=~
      LocalAccelForward=~
      LocalAccelRight=~
      LocalVelocityRight=~
      SkydiveAimPitch=~
      SkydiveAimYaw=~
      DeployChuteAnimRate=~
      SkydiveDriftAnimRate=~
      SkydiveDriftAnimRateCurrent=~
      SkydiveFidgetAnimRate=~
      SkydiveFidgetAnimRateCurrent=~
      SkydiveAdditiveAlpha=~
      SkydiveDriftDelay=~
     ).bIsParachuteOpen
     B=Get bForceOpen()
    ))
  ]
 }
 Transition {
  path=AnimGraph.AnimGraphNode_StateMachine_3.FE Locomotion.AnimStateTransitionNode_2.Transition
  nodes [
   n1 = Get bForceOpen()
   n2 = Break Fort Anim Input Skydiving(FortAnimInput_Skydiving=Get SkydivingState()) -> (
    ParaGilderRootOffset=~
    ParaGliderRootRotationOverride=~
    bUseParaGlideRootModifier=~
    bIsSkydivingFromLaunchPad=~
    bIsSkydivingFromBus=~
    bIsInVortex=~
    bIsDiving=~
    bIsParachuteOpen=~
    bIsSkydiving=~
    bIsParachuteLeaning=~
    bIsSkydiveLeaning=~
    bIsLeaning=~
    bIsSkydiveDiveMode=~
    bPlaySkydiveDrift=~
    bSkydiveDriftDelayActive=~
    bSkydiveDriftAnimAllowed=~
    bIsGliderRight=~
    bIsGliderCenter=~
    bIsGliderLeft=~
    bIsGliderForward=~
    bIsGliderBack=~
    bHasDeployedAndIsStandingUpright=~
    LocalAccelForward=~
    LocalAccelRight=~
    LocalVelocityRight=~
    SkydiveAimPitch=~
    SkydiveAimYaw=~
    DeployChuteAnimRate=~
    SkydiveDriftAnimRate=~
    SkydiveDriftAnimRateCurrent=~
    SkydiveFidgetAnimRate=~
    SkydiveFidgetAnimRateCurrent=~
    SkydiveAdditiveAlpha=~
    SkydiveDriftDelay=~
   )
   Result(bCanEnterTransition=AND Boolean(A=NOT Boolean(A=@n2.bIsParachuteOpen), B=NOT Boolean(A=@n1)))
  ]
 }
 Transition {
  path=AnimGraph.AnimGraphNode_StateMachine_3.FE Locomotion.AnimStateTransitionNode_3.Transition
  nodes [
   n1 = Get bForceOpen()
   n2 = Break Fort Anim Input Skydiving(FortAnimInput_Skydiving=Get SkydivingState()) -> (
    ParaGilderRootOffset=~
    ParaGliderRootRotationOverride=~
    bUseParaGlideRootModifier=~
    bIsSkydivingFromLaunchPad=~
    bIsSkydivingFromBus=~
    bIsInVortex=~
    bIsDiving=~
    bIsParachuteOpen=~
    bIsSkydiving=~
    bIsParachuteLeaning=~
    bIsSkydiveLeaning=~
    bIsLeaning=~
    bIsSkydiveDiveMode=~
    bPlaySkydiveDrift=~
    bSkydiveDriftDelayActive=~
    bSkydiveDriftAnimAllowed=~
    bIsGliderRight=~
    bIsGliderCenter=~
    bIsGliderLeft=~
    bIsGliderForward=~
    bIsGliderBack=~
    bHasDeployedAndIsStandingUpright=~
    LocalAccelForward=~
    LocalAccelRight=~
    LocalVelocityRight=~
    SkydiveAimPitch=~
    SkydiveAimYaw=~
    DeployChuteAnimRate=~
    SkydiveDriftAnimRate=~
    SkydiveDriftAnimRateCurrent=~
    SkydiveFidgetAnimRate=~
    SkydiveFidgetAnimRateCurrent=~
    SkydiveAdditiveAlpha=~
    SkydiveDriftDelay=~
   )
   Result(bCanEnterTransition=AND Boolean(A=NOT Boolean(A=@n2.bIsParachuteOpen), B=NOT Boolean(A=@n1)))
  ]
 }
 Roar {
  entry=Identity
  path=AnimGraph.AnimGraphNode_StateMachine_4.Roar
  ids=id
  nodes [
   AnimStateNode_0 = Caw() {is=title AnimStateNode}
   AnimStateNode_1 = Identity() {is=title AnimStateNode}
   AnimStateNode_4 = Caw_FE() {is=title AnimStateNode}
   AnimStateTransitionNode_0 = AnimStateTransitionNode() {bAllowInertializationForSelfTransitions=true}
   AnimStateTransitionNode_1 = AnimStateTransitionNode() {bAllowInertializationForSelfTransitions=true, bAutomaticRuleBasedOnSequencePlayerInState=true}
   AnimStateTransitionNode_5 = AnimStateTransitionNode() {bAllowInertializationForSelfTransitions=true}
   AnimStateTransitionNode_6 = AnimStateTransitionNode() {bAllowInertializationForSelfTransitions=true, bAutomaticRuleBasedOnSequencePlayerInState=true}
  ]
  links [
   {dst=Identity, rule=AnimGraph.AnimGraphNode_StateMachine_4.Roar.AnimStateTransitionNode_6.Transition, src=Caw, via=AnimStateTransitionNode_6}
   {
    dst=Identity
    rule=AnimGraph.AnimGraphNode_StateMachine_4.Roar.AnimStateTransitionNode_1.Transition
    src=Caw_FE
    via=AnimStateTransitionNode_1
   }
   {dst=Caw, rule=AnimGraph.AnimGraphNode_StateMachine_4.Roar.AnimStateTransitionNode_5.Transition, src=Identity, via=AnimStateTransitionNode_5}
   {
    dst=Caw_FE
    rule=AnimGraph.AnimGraphNode_StateMachine_4.Roar.AnimStateTransitionNode_0.Transition
    src=Identity
    via=AnimStateTransitionNode_0
   }
  ]
 }
 Caw {
  path=AnimGraph.AnimGraphNode_StateMachine_4.Roar.AnimStateNode_0.Caw
  nodes [Output Animation Pose(Result="Sequence Player (additive) 'OutlineLove_Caw_Anim_GLIDER'"() {bLoopAnimation=false, sequence=&/OutlineLove_Caw_Anim_GLIDER})]
 }
 Identity {path=AnimGraph.AnimGraphNode_StateMachine_4.Roar.AnimStateNode_1.Identity, nodes [Output Animation Pose(Result=Additive Identity Pose())]}
 Caw_FE {
  path=AnimGraph.AnimGraphNode_StateMachine_4.Roar.AnimStateNode_4.Caw_FE
  nodes [
   n1 = "Sequence Player (additive) 'OutlineLove_Caw_FrontEnd_Anim_GLIDER'"() {bLoopAnimation=false, sequence=&/OutlineLove_Caw_FrontEnd_Anim_GLIDER}
   Output Animation Pose(Result=@n1)
  ]
 }
 Transition {
  path=AnimGraph.AnimGraphNode_StateMachine_4.Roar.AnimStateTransitionNode_0.Transition
  nodes [Result(bCanEnterTransition=Get FECaw() -> (FECaw:bool))]
 }
 Transition {path=AnimGraph.AnimGraphNode_StateMachine_4.Roar.AnimStateTransitionNode_1.Transition, nodes [Result()]}
 Transition {
  path=AnimGraph.AnimGraphNode_StateMachine_4.Roar.AnimStateTransitionNode_5.Transition
  nodes [Result(bCanEnterTransition=Get GPCaw() -> (GPCaw:bool))]
 }
 Transition {path=AnimGraph.AnimGraphNode_StateMachine_4.Roar.AnimStateTransitionNode_6.Transition, nodes [Result()]}
 Head Darts {
  entry=Identity
  path=AnimGraph.AnimGraphNode_StateMachine_5.Head Darts
  ids=id
  nodes [
   AnimStateNode_1 = Identity()
   AnimStateNode_2 = Head Dart Left() {is=title AnimStateNode}
   AnimStateNode_3 = Head Dart Right() {is=title AnimStateNode}
   AnimStateTransitionNode_1 = AnimStateTransitionNode() {bAllowInertializationForSelfTransitions=true}
   AnimStateTransitionNode_2 = AnimStateTransitionNode() {bAllowInertializationForSelfTransitions=true}
   AnimStateTransitionNode_3 = AnimStateTransitionNode() {bAllowInertializationForSelfTransitions=true, bAutomaticRuleBasedOnSequencePlayerInState=true}
   AnimStateTransitionNode_4 = AnimStateTransitionNode() {bAllowInertializationForSelfTransitions=true, bAutomaticRuleBasedOnSequencePlayerInState=true}
  ]
  links [
   {
    dst=Identity
    rule=AnimGraph.AnimGraphNode_StateMachine_5.Head Darts.AnimStateTransitionNode_3.Transition
    src=Head Dart Left
    via=AnimStateTransitionNode_3
   }
   {
    dst=Identity
    rule=AnimGraph.AnimGraphNode_StateMachine_5.Head Darts.AnimStateTransitionNode_4.Transition
    src=Head Dart Right
    via=AnimStateTransitionNode_4
   }
   {
    dst=Head Dart Left
    rule=AnimGraph.AnimGraphNode_StateMachine_5.Head Darts.AnimStateTransitionNode_1.Transition
    src=Identity
    via=AnimStateTransitionNode_1
   }
   {
    dst=Head Dart Right
    rule=AnimGraph.AnimGraphNode_StateMachine_5.Head Darts.AnimStateTransitionNode_2.Transition
    src=Identity
    via=AnimStateTransitionNode_2
   }
  ]
 }
 Identity {
  path=AnimGraph.AnimGraphNode_StateMachine_5.Head Darts.AnimStateNode_1.Identity
  nodes [Output Animation Pose(Result=Additive Identity Pose())]
 }
 Head Dart Left {
  path=AnimGraph.AnimGraphNode_StateMachine_5.Head Darts.AnimStateNode_2.Head Dart Left
  nodes [
   n1 = "Sequence Player (additive) 'OutlineLove_Head_Dart_Left_Anim_GLIDER'"() {bLoopAnimation=false, sequence=&/OutlineLove_Head_Dart_Left_Anim_GLIDER}
   Output Animation Pose(Result=@n1)
  ]
 }
 Head Dart Right {
  path=AnimGraph.AnimGraphNode_StateMachine_5.Head Darts.AnimStateNode_3.Head Dart Right
  nodes [
   n1 = "Sequence Player (additive) 'OutlineLove_Head_Dart_Right_Anim_GLIDER'"() {bLoopAnimation=false, sequence=&/OutlineLove_Head_Dart_Right_Anim_GLIDER}
   Output Animation Pose(Result=@n1)
  ]
 }
 Transition {
  path=AnimGraph.AnimGraphNode_StateMachine_5.Head Darts.AnimStateTransitionNode_1.Transition
  nodes [Result(bCanEnterTransition=Get LeftHeadDart() -> (LeftHeadDart:bool))]
 }
 Transition {
  path=AnimGraph.AnimGraphNode_StateMachine_5.Head Darts.AnimStateTransitionNode_2.Transition
  nodes [Result(bCanEnterTransition=Get RightHeadDart() -> (RightHeadDart:bool))]
 }
 Transition {path=AnimGraph.AnimGraphNode_StateMachine_5.Head Darts.AnimStateTransitionNode_3.Transition, nodes [Result()]}
 Transition {path=AnimGraph.AnimGraphNode_StateMachine_5.Head Darts.AnimStateTransitionNode_4.Transition, nodes [Result()]}
 EventGraph {nodes [
   n1 = Event BlueprintUpdateAnimation()
   > Set Time(Time:double=@n1.DeltaTimeX) -> (Output_Get:double)
   > n2 = Sequence!(execute) -> (then_0, then_1, then_2:Exec, then_3:Exec)
   > n3 = ForwardBack_Value_Interp!(
    execute
    Time:double=Get Time() -> (Time:double)
    Currnet_Value:double=Get Jitter() -> (Jitter:double)
    Idle_Value:double=1
    Idle_InterpSpeed:double=1
    Forward_Value:double=1.2
    Forward_InterpSpeed:double=1
    Back_Value:double=0.75
    Back_InterpSpeed:double=2
    LocalAccelForward:double=Break Fort Anim Input Skydiving(FortAnimInput_Skydiving=Get SkydivingState()) -> (
     ParaGilderRootOffset=~
     ParaGliderRootRotationOverride=~
     bUseParaGlideRootModifier=~
     bIsSkydivingFromLaunchPad=~
     bIsSkydivingFromBus=~
     bIsInVortex=~
     bIsDiving=~
     bIsParachuteOpen=~
     bIsSkydiving=~
     bIsParachuteLeaning=~
     bIsSkydiveLeaning=~
     bIsLeaning=~
     bIsSkydiveDiveMode=~
     bPlaySkydiveDrift=~
     bSkydiveDriftDelayActive=~
     bSkydiveDriftAnimAllowed=~
     bIsGliderRight=~
     bIsGliderCenter=~
     bIsGliderLeft=~
     bIsGliderForward=~
     bIsGliderBack=~
     bHasDeployedAndIsStandingUpright=~
     LocalAccelForward=~
     LocalAccelRight=~
     LocalVelocityRight=~
     SkydiveAimPitch=~
     SkydiveAimYaw=~
     DeployChuteAnimRate=~
     SkydiveDriftAnimRate=~
     SkydiveDriftAnimRateCurrent=~
     SkydiveFidgetAnimRate=~
     SkydiveFidgetAnimRateCurrent=~
     SkydiveAdditiveAlpha=~
     SkydiveDriftDelay=~
    ).LocalAccelForward
   ) -> (then, New_Value:double)
   > Set Jitter(Jitter:double=@n3) -> (Output_Get:double)
   Branch(
    execute=@n2.then_2
    Condition=AND Boolean!(
     A=Get IntoEnd() -> (IntoEnd:bool)
     B=Get EnableHeadDarts() -> (EnableHeadDarts:bool)
     C:bool=Get EnableCaw() -> (EnableCaw:bool)
     D:bool=Get GetFullyDeployed() -> (GetFullyDeployed:bool)
    ) -> (ReturnValue)
   )
   > Delay(Duration=RandomFloatInRange(Min=10, Max=25))
   > n4 = Sequence!(execute) -> (then_0, then_1)
   n5 = Set Random Integer(execute=[@n4.then_0, @n6.else], Random Integer:int=RandomIntegerInRange!(Min:int=0, Max:int=2) -> (ReturnValue:int=0)) -> (Output_Get:int=0)
   > n6 = Branch(Condition="Not Equal (Integer)"(A=@n5, B=Get Applied Integer() -> (Applied Integer:int=0)))
   > Set Applied Integer(Applied Integer:int=Get Random Integer() -> (Random Integer:int=0)) -> (Output_Get:int=0)
   > n7 = Branch(Condition=Get FrontEnd() -> (FrontEnd:bool))
   n8 = Switch on Int!(execute:Exec=[@n7.then, @n11.else], Selection:int=Get Applied Integer() -> (Applied Integer=0)) -> (Default:Exec, 0:Exec, 1:Exec, 2:Exec)
   Branch(execute=@n8.0, Condition=AND Boolean!(A=Get EnableHeadDarts(), B=Get IntoEnd()) -> (ReturnValue))
   > Set EnableHeadDarts(EnableHeadDarts:bool=false) -> (Output_Get:bool)
   > Set RightHeadDart(RightHeadDart:bool=true) -> (Output_Get:bool)
   Branch(execute=@n8.2, Condition=AND Boolean(A=Get EnableCaw(), B=Get IntoEnd()))
   > Set EnableCaw(EnableCaw:bool=false) -> (Output_Get:bool)
   > n9 = Branch(Condition=Get FrontEnd())
   > Set FECaw(FECaw:bool=true) -> (Output_Get:bool)
   Set GPCaw(execute=@n9.else, GPCaw:bool=true) -> (Output_Get:bool)
   Branch(execute=@n8.1, Condition=AND Boolean(A=Get EnableHeadDarts(), B=Get IntoEnd()))
   > Set EnableHeadDarts(EnableHeadDarts=false)
   > Set LeftHeadDart(LeftHeadDart:bool=true) -> (Output_Get:bool)
   n10 = Break Fort Anim Input Skydiving(FortAnimInput_Skydiving=Get SkydivingState()) -> (
    ParaGilderRootOffset=~
    ParaGliderRootRotationOverride=~
    bUseParaGlideRootModifier=~
    bIsSkydivingFromLaunchPad=~
    bIsSkydivingFromBus=~
    bIsInVortex=~
    bIsDiving=~
    bIsParachuteOpen=~
    bIsSkydiving=~
    bIsParachuteLeaning=~
    bIsSkydiveLeaning=~
    bIsLeaning=~
    bIsSkydiveDiveMode=~
    bPlaySkydiveDrift=~
    bSkydiveDriftDelayActive=~
    bSkydiveDriftAnimAllowed=~
    bIsGliderRight=~
    bIsGliderCenter=~
    bIsGliderLeft=~
    bIsGliderForward=~
    bIsGliderBack=~
    bHasDeployedAndIsStandingUpright=~
    LocalAccelForward=~
    LocalAccelRight=~
    LocalVelocityRight=~
    SkydiveAimPitch=~
    SkydiveAimYaw=~
    DeployChuteAnimRate=~
    SkydiveDriftAnimRate=~
    SkydiveDriftAnimRateCurrent=~
    SkydiveFidgetAnimRate=~
    SkydiveFidgetAnimRateCurrent=~
    SkydiveAdditiveAlpha=~
    SkydiveDriftDelay=~
   )
   n11 = Branch(
    execute=@n7.else
    Condition=OR Boolean!(A=@n10.bIsGliderRight, B=@n10.bIsGliderLeft, C:bool=@n10.bIsGliderForward, D:bool=@n10.bIsGliderBack) -> (ReturnValue)
   )
   Event BlueprintInitializeAnimation()
   > n12 = Sequence!(execute) -> (then_0, then_1, then_2:Exec)
   Set EnableHeadDarts(execute=[@n15.then, @n12.then_2, @n14.then_1], EnableHeadDarts=false)
   > Set EnableCaw(EnableCaw=false)
   > Set GetFullyDeployed(GetFullyDeployed:bool=false) -> (Output_Get:bool)
   HideBoneByName!(
    execute=[@n14.then_0, @n12.then_1]
    self:Skinned Mesh Component Object Reference=GetOwningComponent()
    BoneName:Name=Glider_Base
    PhysBodyOption:EPhysBodyOp Enum=PBO_None
   ) -> (then)
   n13 = Branch(execute=@n12.then_0, Condition=IsCosmeticObjectDisplayedInFrontendContext!(CosmeticObject:Object Reference) -> (ReturnValue:bool))
   > Set FrontEnd(FrontEnd:bool=true) -> (Output_Get:bool)
   Set FrontEnd(execute=@n13.else, FrontEnd=false)
   AnimNotify_ClosedEntered() {is=title K2Node_Event}
   > n14 = Sequence!(execute) -> (then_0, then_1)
   AnimNotify_DeployEntered() {is=title K2Node_Event}
   > Delay(Duration=0.1)
   > n15 = UnHideBoneByName!(execute, self:Skinned Mesh Component Object Reference=GetOwningComponent(), BoneName:Name=Glider_Base) -> (then)
   AnimNotify_TurnONEngines() {is=title K2Node_Event}
   > SetGliderFullyDeployed!(execute, self:Fort Glider Anim Instance Object Reference, bIsFullyDeployed:bool=true) -> (then)
   > Set GetFullyDeployed(GetFullyDeployed=true)
   AnimNotify_Into End() {is=title K2Node_Event}
   > Delay(Duration=2)
   > Set IntoEnd(IntoEnd:bool=true) -> (Output_Get:bool)
   > Set EnableCaw(EnableCaw=true)
   > Set EnableHeadDarts(EnableHeadDarts=true)
   Event BlueprintPostEvaluateAnimation() {is=title K2Node_Event}
   > n16 = Sequence!(execute) -> (then_0, then_1, then_2:Exec, then_3:Exec)
   > Set LeftHeadDart(LeftHeadDart=false)
   > Set EnableHeadDarts(EnableHeadDarts=true)
   Set RightHeadDart(execute=@n16.then_1, RightHeadDart=false)
   > Set EnableHeadDarts(EnableHeadDarts=true)
   Set FECaw(execute=@n16.then_2, FECaw=false)
   > Set EnableCaw(EnableCaw=true)
   Set GPCaw(execute=@n16.then_3, GPCaw=false)
   > Set EnableCaw(EnableCaw=true)
   AnimNotify_Into Start() {is=title K2Node_Event}
   > Set IntoEnd(IntoEnd=false)
   > Set EnableCaw(EnableCaw=false)
   > Set EnableHeadDarts(EnableHeadDarts=false)
   n17 = Get OwnerGlider()
   Set Actor Hidden In Game!(execute, self:Actor Object Reference=@n17, bNewHidden:bool=true) -> (then) {is=title K2Node_CallFunction}
   Set Actor Hidden In Game(self=@n17, bNewHidden=true)
  ]}
 GetPlayerOpenGliderBool {nodes [
   Inputs()
   > n1 = Branch(Condition=Get IsCosmeticPreview(self=Get OwnerGlider()))
   Outputs!(then:Exec=[@n3.then, @n1.then])
   n2 = Get OwnerFortPawn() -> (OwnerFortPawn:Fort Player Pawn Object Reference)
   n3 = Set bForceOpen(
    execute=@n1.else
    bForceOpen:bool=ShouldForceOpenGlider!(self:Fort Player Animation Interface Object Reference=GetAnimationInterface!(self:Fort Player Pawn Object Reference=@n2) -> (ReturnValue:Fort Player Animation Interface Object Reference)) -> (ReturnValue:bool)
   ) -> (Output_Get:bool)
  ]}
]
variables [
 Applied Integer {type=integer}
 EnableCaw {type=boolean}
 EnableHeadDarts {type=boolean}
 FECaw {type=boolean}
 FrontEnd {type=boolean}
 GetFullyDeployed {type=boolean}
 GPCaw {type=boolean}
 IntoEnd {type=boolean}
 Jitter {type=number}
 LeftHeadDart {type=boolean}
 PlayersLeaning {type=boolean}
 Random Integer {type=integer}
 RightHeadDart {type=boolean}
 Time {type=number}
 Turbine {type=number}
]
